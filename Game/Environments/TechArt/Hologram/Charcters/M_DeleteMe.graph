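# M_DeleteMe is a Default Lit material in the Hologram characters TechArt folder, apparently a throwaway placeholder given its name.
# Its graph holds a single magenta-pink Constant3Vector, with no connection to any material output shown.
# Automatic usage flagging in the editor is turned off, and pixel depth offset uses legacy mode.
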Material /Game/Environments/TechArt/Hologram/Charcters/M_DeleteMe
graph {
 properties {bAutomaticallySetUsageInEditor=false, pixelDepthOffsetMode=PDOM_Legacy, usedShadingModels=Default Lit}
 nodes [Constant3Vector() {constant=<rgba 1 0 0.8799 1>}]
}
